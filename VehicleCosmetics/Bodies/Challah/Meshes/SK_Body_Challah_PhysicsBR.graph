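# Physics asset for the Challah vehicle body skeletal mesh, a cosmetic body; the BR suffix marks it as a variant.
# It holds a single root body and no constraints, so the mesh simulates as one rigid piece.
# The root body lists no collision shapes in this data.
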PhysicsAsset /VehicleCosmetics/Bodies/Challah/Meshes/SK_Body_Challah_PhysicsBR
bodies {root []}
constraints []
